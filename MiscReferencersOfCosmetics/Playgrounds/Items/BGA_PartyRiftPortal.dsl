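; Party Signal rift portal actor that players interact with to join or leave the owner's party, or remove their own signal.
; BeginPlay destroys it if a scalable float disables spawning, otherwise async-loads soft-referenced info-plate and interaction widgets.
; Party and owner events refresh those widgets, while range or long-range requests control plate visibility.
; Timelines and overlaps animate the rift with audio, and a ready thumbnail replaces the rift mesh.
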(event EventBeginPlay
  ;[1 1 1] EVENT--> BeginPlay
  (|CallParentFunction)
  (Variables|FortAthenaPartyRiftPortal|SetInteractIconPlacementCapsule (Variables|Default|GetOverlapCapsule))
  (if (> (ScalableFloat|GetValueatLevel (Variables|Default|GetShouldSpawnRift?)) 0.0)
    (Utilities|FlowControl|Sequence
      (:then_0
        (CallFunction|SetInteractionInteractComponent (Variables|Default|GetCapsule))
        (Audio|Components|Audio|SetVolumeMultiplier (Variables|Sound|GetAudio_Looping) 0.5 :!class /Script/Engine.AudioComponent)
        (Components|Sphere|SetSphereRadius (Variables|Default|GetPostProcessRadius) 25.0 false)
        (Variables|UI|SetPortalInfoPlateOriginalOffset (.z (Variables|/Script/Engine.SceneComponent|GetRelativeLocation (Variables|Default|GetUI_PortalInfoPlate))))
        (Rendering|SetVisibility (Variables|Default|GetSM_Rift) (Variables|Default|GetbShowSM_Rift) true))
      (:then_1
        (Networking|CanExecuteCosmeticEvents
          (:True
            (UI|UpdateInfoPlatePointer)
            ;;[0 0 0] setup soft widget ref and init through interface
            (Utilities|FlowControl|Sequence
              (:then_0
                (bind _class (Utilities|AsyncLoadClassAsset (Variables|UI|GetUI_InteractionDisplayWidgetSoftClass))
                  (:Completed
                    (bind _asuser_widget (Utilities|Casting|CastToUserWidgetClass _class))
                    (bind _returnvalue (UserInterface|CreateWidget _asuser_widget (Fortnite|GetFirstLocalFortPlayerController true)))
                    (UI|InitializeInteractionWidgetInfo _returnvalue)
                    (CallFunction|SetInteractionDisplayWidget _returnvalue)
                    (Init|InitfromObject(Message) _returnvalue self)
                    (Variables|Interaction|SetbShowInteractKeybind true)
                    (Variables|Interaction|SetbAllowInteract true))))
              (:then_1
                (bind _class_2 (Utilities|AsyncLoadClassAsset (Variables|UI|GetUI_PortalInfoPlateSoftClass))
                  (:Completed
                    (bind _asuser_widget_2 (Utilities|Casting|CastToUserWidgetClass _class_2))
                    (bind _returnvalue_2 (UserInterface|CreateWidget _asuser_widget_2 (Fortnite|GetFirstLocalFortPlayerController true)))
                    (UI|InitalizePartyRiftInfo _returnvalue_2)
                    (UserInterface|SetWidget (Variables|Default|GetUI_PortalInfoPlate) _returnvalue_2)
                    (Init|InitfromObject(Message) _returnvalue_2 self)
                    (UI|UpdateOwnerDisplay)
                    (UI|UpdatePopulationDisplay)
                    (CallFunction|CheckShouldDisplayUI)
                    (CallFunction|UpdatePortal)))))))))
    (else
      (Actor|DestroyActor))))

(event EventAdjustBeaconCollisionHeight
  (params (Distance :type Float%20%28single-precision%29))
  (bind _returnvalue (* Distance 0.5))
  (bind _beaconcollision (Variables|Default|GetBeaconCollision))
  (Components|Capsule|SetCapsuleHalfHeight _beaconcollision _returnvalue false)
  (Transformation|SetRelativeLocation _beaconcollision (Math|Vector|MakeVector 0.0 0.0 _returnvalue) false false))

(event EventShowLongRangeInteractionUI
  (params (bShow :type Boolean) (InteractionOffset :type Float%20%28single-precision%29))
  (Variables|UI|SetLongRangeUIVisibility bShow)
  (Variables|UI|SetLongRangeUIIntersectionOffset InteractionOffset)
  (CallFunction|CheckShouldDisplayUI))

(event Custom|CheckShouldDisplayUI
  ;[0 0 0] Show or hide the portal ui info plate
  (UI|SetPortalUIVisibility)
  (CallFunction|UpdatePartyJoinInfo))

(event EventOwningPlayerNetIdChanged
  ;[0 0 0] wip owner setting
  (bind _returnvalue (UniqueNetID|IsValidNetID (Variables|Portal|GetOwningPlayerNetId)))
  (if _returnvalue
    (UI|UpdateOwnerDisplay)
    (else
      (Actor|DestroyActor))))

(event EventBroadcastIsBeingInteractedWith
  (params (IsInteracting :type Boolean)))

(event Custom|UpdatePortal
  (bind _sm_rift (Variables|Default|GetSM_Rift))
  (Networking|CanExecuteCosmeticEvents
    (:True
      (if (Variables|Default|GetbShowSM_Rift)
        (Rendering|Material|SetMaterial (Variables|Default|GetSM_Rift) 0 "/Game/Athena/Items/ForagedItems/Rift/FX/M_Shard_Portal_Gold_Inst.M_Shard_Portal_Gold_Inst")
        (if (Variables|Portal|GetbPortalOpen)
          (bind (_direction _open) (Timeline|Timeline_0 Play)
            (:Update
              (Transformation|SetRelativeScale3D _sm_rift (Math|Conversions|ToVector(Float) _open))
              (Rendering|Material|SetScalarParameterValueonMaterials _sm_rift "CloseMult" _open)
              (Transformation|SetRelativeRotation _sm_rift (Math|Rotator|MakeRotator (* _open -2e+01)) false false)))
          (else
            (Timeline|Timeline_0 Reverse)))))))

(event OnComponentBeginOverlap(HUDVisibilityVolume)
  (params (OverlappedComponent :type Primitive%20Component%20Object%20Reference) (OtherActor :type Actor%20Object%20Reference) (OtherComp :type Primitive%20Component%20Object%20Reference) (OtherBodyIndex :type Integer) (bFromSweep :type Boolean) (SweepResult :type Hit%20Result%20Structure%20%28by%20ref%29))
  ;[1 1 1] Start Detecting Looking at the Code Screen or not
  (CodeOverride|LocalPlayerRangeCheck OtherActor true))

(event Custom|EvaluateOverlap
  (bind _sm_rift (Variables|Default|GetSM_Rift))
  ;[1 1 1] On Overlap - Toggle Post Process
  (if (> (Utilities|Array|Length (CollectPawns|GetOverlappingPlayerPawns self)) 0)
    (if (Variables|Default|GetbShowSM_Rift)
      ;;[1 1 1] Widen rift if player is close
      (bind (_direction _open) (Timeline|VectorScale Play)
        (:Update
          (Rendering|Material|SetScalarParameterValueonMaterials _sm_rift "Open" (Math|Float|SelectFloat _open 0.0 (Variables|Portal|GetbPortalOpen)))
          (Transformation|SetRelativeRotation _sm_rift (Math|Rotator|MakeRotator (Math|Float|SelectFloat (* _open -2e+01) 0.0 (Variables|Portal|GetbPortalOpen))) false false))))
    (else
      (if (Variables|Default|GetbShowSM_Rift)
        (Timeline|VectorScale Reverse)))))

(event OnComponentEndOverlap(HUDVisibilityVolume)
  (params (OverlappedComponent :type Primitive%20Component%20Object%20Reference) (OtherActor :type Actor%20Object%20Reference) (OtherComp :type Primitive%20Component%20Object%20Reference) (OtherBodyIndex :type Integer))
  ;[1 1 1] Start Detecting Looking at the Code Screen or not
  (CodeOverride|LocalPlayerRangeCheck OtherActor))

(event OnComponentBeginOverlap(Capsule)
  (params (OverlappedComponent :type Primitive%20Component%20Object%20Reference) (OtherActor :type Actor%20Object%20Reference) (OtherComp :type Primitive%20Component%20Object%20Reference) (OtherBodyIndex :type Integer) (bFromSweep :type Boolean) (SweepResult :type Hit%20Result%20Structure%20%28by%20ref%29))
  (bind _sm_rift (Variables|Default|GetSM_Rift))
  ;[1 1 1] On Overlap - Toggle Post Process
  (Utilities|IsValid (CollectPawns|GetPlayerPawnfromActor OtherActor)
    (:Is%20Valid
      (CallFunction|PlayExpandorCollapseAudio true)
      (if (Variables|Default|GetbShowSM_Rift)
        ;;[1 1 1] Widen rift if player is close
        (Timeline|VectorScale Play)))))

(event OnComponentEndOverlap(Capsule)
  (params (OverlappedComponent :type Primitive%20Component%20Object%20Reference) (OtherActor :type Actor%20Object%20Reference) (OtherComp :type Primitive%20Component%20Object%20Reference) (OtherBodyIndex :type Integer))
  (bind _sm_rift (Variables|Default|GetSM_Rift))
  ;[1 1 1] On Overlap - Toggle Post Process
  (if (<= (Utilities|Array|Length (CollectPawns|GetOverlappingPlayerPawns OtherActor)) 0)
    (CallFunction|PlayExpandorCollapseAudio false)
    (if (Variables|Default|GetbShowSM_Rift)
      ;;[1 1 1] Widen rift if player is close
      (Timeline|VectorScale Reverse))))

(event EventOnMaxPartySizeChanged
  (UI|UpdatePopulationDisplay))

(event EventOnPartyMembersChanged
  (UI|UpdatePopulationDisplay))

(event EventPortalOpenChanged
  (Networking|CanExecuteCosmeticEvents
    (:True
      (CallFunction|PlayExpandorCollapseAudio (Variables|Portal|GetbPortalOpen)))
    (:False
      _))
  (CallFunction|UpdatePortal))

(event EventErrorStateChanged
  (CallFunction|UpdatePortal))

(event Custom|Play%20Rift%20Exit%20Sound
  (params (Location :type Vector))
  ;[1 1 1] Lets others know someone is rifting in
  (Audio|SpawnSoundatLocation (Variables|Sound|GetPlayerExitSkyCrack) Location))

(event EventOnThumbnailTextureReady
  (bind _thumbnailplane (Variables|Default|GetThumbnailPlane))
  ;[1 1 1] Thumbnails
  (Utilities|IsValid (Variables|FortAthenaPartyRiftPortal|GetThumbnailTexture)
    (:Is%20Valid
      (Development|SetHiddeninGame (Variables|Default|GetThumbnailPlane) false false)
      (bind _returnvalue (Rendering|Material|CreateDynamicMaterialInstance (Variables|Default|GetThumbnailPlane) 0 0 "None" :!class /Script/Engine.PrimitiveComponent))
      (Rendering|Material|SetTextureParameterValue _returnvalue "Texture" (Variables|FortAthenaPartyRiftPortal|GetThumbnailTexture))
      (Rendering|Material|SetVectorParameterValue _returnvalue "Color" (Variables|Default|GetOwnedThumbnailColor) :!class /Script/Engine.MaterialInstanceDynamic)
      (Transformation|SetAbsolute _thumbnailplane false true true)
      (Transformation|SetWorldRotation _thumbnailplane "0,-90.000000, 0" false false)
      (Transformation|SetWorldScale3D (Variables|Default|GetThumbnailPlane) (CallFunction|ScaleThumbnailPlanebyTextureSize))
      (CallFunction|SetupPortalforThumbnailDisplay))
    (:Is%20Not%20Valid
      (Development|SetHiddeninGame (Variables|Default|GetThumbnailPlane) true false))))

(event Custom|SetupPortalForThumbnailDisplay
  ;[0 0 0] the configuration of the portal when we have a thumbnail available
  (Transformation|SetRelativeLocation (Variables|Default|GetThumbnailPlane) "0, 0,114.000000" false false)
  (Development|SetHiddeninGame (Variables|Default|GetSM_Rift) true false))

(fn BlueprintCanInteract
  (params (InteractingPawn :type Fort%20Pawn%20Object%20Reference) (InteractionBeingAttempted :type EInteractionBeingAttempted%20Enum :default "FirstInteraction") (InteractionType :type TInteractionType%20Enum :default "IT_NoInteraction"))
  (bind _returnvalue (Fortnite|GetFortPlayerControllerfromActor InteractingPawn))
  (Utilities|IsValid InteractingPawn
    (:Is%20Valid
      (Utilities|IsValid _returnvalue
        (:Is%20Valid
          ;[0 0.25 0.25] First interaction is enough
          (if (Utilities|Enum|Equal(Enum) InteractionBeingAttempted "FirstInteraction")
            (return (and (CallFunction|IsPortalOpen) (Math|Boolean|ORBoolean (CallFunction|DoesPlayerOwnPortal _returnvalue) (CallFunction|IsPlayerinOwnerParty _returnvalue) (CallFunction|IsPartyJoinable) (CallFunction|IsPartyJoinRequestable))))
            (else
              _)))
        (:Is%20Not%20Valid
          _))
      (return false))
    (:Is%20Not%20Valid
      _)))

(fn BlueprintDisplayMobileInteractPrompt
  (return (Interaction|MobileInteractIsValid)))

(fn BlueprintGetFailedInteractionString
  (params (InteractingPawn :type Fort%20Pawn%20Object%20Reference) (InteractionBeingAttempted :type EInteractionBeingAttempted%20Enum :default "FirstInteraction"))
  (return (CallFunction|GetPartyJoinPrompt)))

(fn BlueprintGetInteractionString
  (params (InteractingPawn :type Fort%20Pawn%20Object%20Reference) (InteractionBeingAttempted :type EInteractionBeingAttempted%20Enum :default "FirstInteraction"))
  (bind _returnvalue (Fortnite|GetFortPlayerControllerfromActor InteractingPawn))
  (Utilities|IsValid _returnvalue
    (:Is%20Valid
      (if (CallFunction|DoesPlayerOwnPortal _returnvalue)
        (return "NSLOCTEXT(\"[2972018B48B77761F06F30BB06D3BFF2]\", \"07F3A459405D45EBF73AD481D465ADDA\", \"REMOVE SIGNAL\")")
        (elif (CallFunction|IsPlayerinOwnerParty _returnvalue)
          (return "NSLOCTEXT(\"[2972018B48B77761F06F30BB06D3BFF2]\", \"4FA5816541955722FD7C30AF80DD88C1\", \"LEAVE PARTY\")")
          (else
            (return (CallFunction|GetPartyJoinPrompt))))))
    (:Is%20Not%20Valid
      (return ""))))

(fn BlueprintGetInteractionTime
  (params (InteractingPawn :type Fort%20Pawn%20Object%20Reference) (InteractionBeingAttempted :type EInteractionBeingAttempted%20Enum :default "FirstInteraction"))
  (bind _returnvalue (Fortnite|GetFortPlayerControllerfromActor InteractingPawn))
  (Utilities|IsValid _returnvalue
    (:Is%20Valid
      (if (CallFunction|DoesPlayerOwnPortal _returnvalue)
        (return 0.0 false)
        (else
          (return 3.0 true))))
    (:Is%20Not%20Valid
      (return 0.0 false))))

(fn BlueprintOnLocalInteract
  (params (InteractingPawn :type Fort%20Player%20Pawn%20Object%20Reference) (InteractionBeingAttempted :type EInteractionBeingAttempted%20Enum :default "FirstInteraction"))
  (bind _asathena_player_controller (Utilities|Casting|CastToAthena_PlayerController (Fortnite|GetFortPlayerControllerfromActor InteractingPawn))
    (:then
      (bind _returnvalue (CallFunction|OnInteractwithPartyRiftPortal _asathena_player_controller))
      (return _returnvalue))
    (:CastFailed
      (return false))))

(fn BlueprintUseMobileHighlight
  (return (Interaction|MobileInteractIsValid)))

(fn DoesPlayerOwnPortal
  (params (Player :type Fort%20Player%20Controller%20Object%20Reference))
  (return (Utilities|Equal(Object) Player (Actor|GetOwner))))

(fn GetOverlappingPlayerPawns
  (params (ActorPtr :type Actor%20Object%20Reference))
  (for _array_element (Collision|GetOverlappingActors ActorPtr "/Script/FortniteGame.FortPlayerPawn")
    (bind _asfort_player_pawn (Utilities|Casting|CastToFortPlayerPawn _array_element))
    (Utilities|Array|AddUnique (LocalVariables|GetPlayer%20Pawns) _asfort_player_pawn))
  (return (LocalVariables|GetPlayer%20Pawns)))

(fn GetPlayerPawnFromActor
  (params (ActorPtr :type Actor%20Object%20Reference))
  (bind _asfort_player_pawn (Utilities|Casting|CastToFortPlayerPawn ActorPtr)
    (:then
      (return _asfort_player_pawn))
    (:CastFailed
      (return 0))))

(fn InitalizePartyRiftInfo
  (params (Widget :type User%20Widget%20Object%20Reference))
  (bind _asportal_info_plate (Utilities|Casting|CastToPortalInfoPlate Widget))
  (bind _output_get (Variables|UI|SetPortalInfoWidget _asportal_info_plate))
  (Class|PortalInfoPlate|SetIslandName _output_get "NSLOCTEXT(\"[2972018B48B77761F06F30BB06D3BFF2]\", \"AB46785F495D01C1E4CD38B81680F1C9\", \"Party Signal\")" :!class /Game/Playgrounds/Items/Portals/UI/PortalInfoPlate.PortalInfoPlate_C)
  (Class|PortalInfoPlate|SetDescription _output_get "NSLOCTEXT(\"[2972018B48B77761F06F30BB06D3BFF2]\", \"4281EDC641CB2D775D892298BB3125A3\", \"Approach Control Panel To Use\")"))

(fn InitializeInteractionWidgetInfo
  (params (Widget :type User%20Widget%20Object%20Reference))
  (bind _ascreative_portal_hudswitcher (Utilities|Casting|CastToCreativePortalHUDSwitcher Widget))
  (bind _output_get (Variables|UI|SetPortalInteractionWidget (Variables|/Game/Creative/Devices/Common/UI/MatchmakingPortal/CreativePortalHUDSwitcher.CreativePortalHUDSwitcher_C|GetCreativePortalHUDDetails _ascreative_portal_hudswitcher)))
  (Class|CreativePortalHUDDetails|SetCuratedStatus _output_get false false false)
  (Class|CreativePortalHUDDetails|SetIslandName _output_get "NSLOCTEXT(\"[2972018B48B77761F06F30BB06D3BFF2]\", \"B966B3044A9212C96B3839BB2D511CC2\", \"PARTY SIGNAL\")" :!class /Game/Creative/Devices/Common/UI/MatchmakingPortal/CreativePortalHUDDetails.CreativePortalHUDDetails_C)
  (Class|CreativePortalHUDDetails|SetLinkCode _output_get "")
  (Class|CreativePortalHUDDetails|SetDescription _output_get ""))

(fn LocalPlayerRangeCheck
  (params (OverlappingActor :type Actor%20Object%20Reference) (bEnteredRange :type Boolean))
  (bind _returnvalue (Fortnite|GetFortPlayerControllerfromActor OverlappingActor))
  (Variables|FortAthenaPartyRiftPortal|SetbLocalPlayerInRange false)
  (Utilities|FlowControl|Sequence
    (:then_0
      (Networking|CanExecuteCosmeticEvents
        (:True
          (Utilities|IsValid _returnvalue
            (:Is%20Valid
              (if (Utilities|Array|ContainsItem (Fortnite|GetLocalFortPlayerControllers true) _returnvalue)
                (Variables|FortAthenaPartyRiftPortal|SetbLocalPlayerInRange (Variables|GetbEnteredRange))))))))
    (:then_1
      (CallFunction|CheckShouldDisplayUI))))

(fn MobileInteractIsValid
  (return false))

(fn PlayExpandOrCollapseAudio
  (params (Expand :type Boolean))
  (if (and (Variables|Portal|GetbPortalOpen) (Variables|Default|GetbShowSM_Rift))
    (Audio|PlaySoundatLocation (select Expand "/Game/Sounds/Fort_GamePlay_Sounds/Player/RiftPortal/RiftPortal_Expand_Cue.RiftPortal_Expand_Cue" "/Game/Sounds/Fort_GamePlay_Sounds/Player/RiftPortal/RiftPortal_Collapse_Cue.RiftPortal_Collapse_Cue") (Transformation|GetWorldLocation (Variables|Default|GetSM_Rift)))))

(fn ScaleThumbnailPlaneByTextureSize
  (return (Utilities|Select (Math|Vector|MakeVector (/ (Variables|Default|GetThumbnailWorldSizeScale) (/ (Math|Conversions|ToFloat(Integer) (CallFunction|GetThumbnailWidth)) (Math|Conversions|ToFloat(Integer) (CallFunction|GetThumbnailHeight)))) (Variables|Default|GetThumbnailWorldSizeScale) 1.0) (Math|Vector|MakeVector (Variables|Default|GetThumbnailWorldSizeScale) (/ (Variables|Default|GetThumbnailWorldSizeScale) (/ (Math|Conversions|ToFloat(Integer) (CallFunction|GetThumbnailWidth)) (Math|Conversions|ToFloat(Integer) (CallFunction|GetThumbnailHeight)))) 1.0) (> (CallFunction|GetThumbnailWidth) (CallFunction|GetThumbnailHeight)))))

(fn SetPortalUIVisibility
  (bind _ui_portalinfoplate (Variables|Default|GetUI_PortalInfoPlate))
  (bind _longrangeuivisibility (Variables|UI|GetLongRangeUIVisibility))
  (Development|SetHiddeninGame _ui_portalinfoplate (not (or (Variables|FortAthenaPartyRiftPortal|GetbLocalPlayerInRange) _longrangeuivisibility)) true)
  (Transformation|SetRelativeLocation _ui_portalinfoplate (Math|Vector|MakeVector 0.0 0.0 (select _longrangeuivisibility (Variables|UI|GetLongRangeUIIntersectionOffset) (Variables|UI|GetPortalInfoPlateOriginalOffset))) false false))

(fn UpdateOwnerDisplay
  (bind _playername (LocalVariables|GetPlayerName))
  (bind _returnvalue (Utilities|Text|ToText(String) _playername))
  (bind _returnvalue_2 (CallFunction|GetPortalOwnerPlayerState))
  (bind _portalinteractionwidget (Variables|UI|GetPortalInteractionWidget))
  (bind _portalinfowidget (Variables|UI|GetPortalInfoWidget))
  (Utilities|FlowControl|Sequence
    (:then_0
      (LocalVariables|SetPlayerName "")
      (Utilities|IsValid _returnvalue_2
        (:Is%20Valid
          (LocalVariables|SetPlayerName (PlayerState|GetPlayerName _returnvalue_2)))))
    (:then_1
      (Utilities|IsValid _portalinfowidget
        (:Is%20Valid
          (Class|PortalInfoPlate|SetPlayerName _portalinfowidget _returnvalue :!class /Game/Playgrounds/Items/Portals/UI/PortalInfoPlate.PortalInfoPlate_C))))
    (:then_2
      (Utilities|IsValid _portalinteractionwidget
        (:Is%20Valid
          (Class|CreativePortalHUDDetails|SetPlayerName _portalinteractionwidget _returnvalue :!class /Game/Creative/Devices/Common/UI/MatchmakingPortal/CreativePortalHUDDetails.CreativePortalHUDDetails_C))))))

(fn UpdateInfoPlatePointer
  (CallFunction|SetHUDPointerOrigin (Transformation|GetWorldLocation (Variables|Default|GetCodeScreenLookAtSphere))))

(fn UpdatePopulationDisplay
  (bind _returnvalue (CallFunction|GetCurrentPartySize))
  (bind _returnvalue_2 (CallFunction|GetMaxPartySize))
  (bind _portalinteractionwidget (Variables|UI|GetPortalInteractionWidget))
  (bind _portalinfowidget (Variables|UI|GetPortalInfoWidget))
  (Utilities|FlowControl|Sequence
    (:then_0
      (Utilities|IsValid _portalinfowidget
        (:Is%20Valid
          (Class|PortalInfoPlate|SetIslandPop _portalinfowidget _returnvalue _returnvalue_2))))
    (:then_1
      (Utilities|IsValid _portalinteractionwidget
        (:Is%20Valid
          (Class|CreativePortalHUDDetails|SetIslandPop _portalinteractionwidget _returnvalue _returnvalue_2))))))

(fn ConstructionScript)
